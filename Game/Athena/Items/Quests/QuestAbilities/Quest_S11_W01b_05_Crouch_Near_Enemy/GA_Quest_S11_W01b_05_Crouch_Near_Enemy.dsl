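; Quest gameplay ability for the Season 11 crouch-near-enemy quest, tracking how long the owning player stays near hostile players.
; On activation, a looping timer checks for hostile pawns within a radius, adding time when consecutive checks succeed.
; Whole seconds go to the quest manager as a stat event, and the remainder carries over.
; Player DBNO cancels the ability; ending runs a last check crediting the partial interval via server time.
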(event Ability|EventActivateAbilityFromEvent
  (params (EventData :type Gameplay%20Event%20Data%20Structure%20%28by%20ref%29))
  ;[1 1 1] Run first check and start looping check
  (|CallParentFunction EventData)
  (CallFunction|CheckforNearbyEnemies false)
  (bind _returnvalue (Utilities|Time|SetTimerbyEvent (delegate Custom|LoopCheck) (Variables|Default|GetCheckInterval) true))
  (Variables|Default|SetTimer_CheckEnemies _returnvalue)
  (EventDispatchers|BindEventtoOnPlayerStartDBNO (Variables|GAVars|Data|GetOwningPlayerPawn) (EventDispatchers|CreateEvent :!event K2_CancelAbility)))

(event Custom|LoopCheck
  ;[1 1 1] Run first check and start looping check
  (CallFunction|CheckforNearbyEnemies false))

(event Custom|CheckForNearbyEnemies
  (params (LastCheck :type Boolean))
  (bind _accumulatedsuccesstime (Variables|Default|GetAccumulatedSuccessTime))
  (bind _returnvalue (Math|Float|Floor _accumulatedsuccesstime))
  ;[1 1 1] Handle checking for nearby enemies
  ;;[1 1 1] If not on the last check, save the time this loop triggered (used by last check)
  (if (not LastCheck)
    (Variables|Default|SetLastLoopTime (Fortnite|Sync|Time|GetServerWorldTimeSeconds)))
  (Variables|Default|SetCurrentCheckSucceeded false)
  ;;[1 1 1] Check for nearby Fort Player Pawns
  (bind _outfortplayerpawns (Fortnite|GetFortPlayerPawnsInRange (Transformation|GetActorLocation (Variables|GAVars|Data|GetOwningPlayerPawn)) (* (Variables|Default|GetRadiusInMeters) 1e+02)))
  ;;[1 1 1] Check each hit actor until we find an enemy
  (bind _array_element (Utilities|Array|ForEachLoopwithBreak _outfortplayerpawns)
    (:LoopBody
      (if (Utilities|Enum|Equal(Enum) (Team|GetTeamAffiliation _array_element (Variables|GAVars|Data|GetOwningPlayerPawn)) "Hostile")
        (Variables|Default|SetCurrentCheckSucceeded true)))
    (:Completed
      _))
  (if (and (Variables|Default|GetLastCheckSucceeded) (Variables|Default|GetCurrentCheckSucceeded))
    (Variables|Default|SetAccumulatedSuccessTime (+ (Variables|Default|GetAccumulatedSuccessTime) (select LastCheck (- (Fortnite|Sync|Time|GetServerWorldTimeSeconds) (Variables|Default|GetLastLoopTime)) (Variables|Default|GetCheckInterval))) :!pinned true) ; Add the successful time
    (else
      _))
  (Variables|Default|SetLastCheckSucceeded (Variables|Default|GetCurrentCheckSucceeded) :!pinned true) ; Save off whether this check was successful
  ;;[1 1 1] Update quest progress as we exceed int threshold, save off leftover successful time
  (if (>= (Variables|Default|GetAccumulatedSuccessTime) 1.0)
    (Quest|SendComplexCustomStatEvent (Variables|GAVars|Data|GetOwningPlayerQuestManager) 0 (Variables|Default|GetQuestTag) "(GameplayTags=)" _returnvalue)
    (Variables|Default|SetAccumulatedSuccessTime (- _accumulatedsuccesstime (Math|Conversions|ToFloat(Integer) _returnvalue)))))

(event Ability|EventOnEndAbility
  (params (bWasCancelled :type Boolean))
  ;[1 1 1] OnEndAbility: Kill timer and run one last check
  (|CallParentFunction bWasCancelled)
  (Utilities|Time|ClearandInvalidateTimerbyHandle (Variables|Default|GetTimer_CheckEnemies))
  (CallFunction|CheckforNearbyEnemies true)
  (EventDispatchers|UnbindEventfromOnPlayerStartDBNO (Variables|GAVars|Data|GetOwningPlayerPawn) (EventDispatchers|CreateEvent :!event K2_CancelAbility)))
